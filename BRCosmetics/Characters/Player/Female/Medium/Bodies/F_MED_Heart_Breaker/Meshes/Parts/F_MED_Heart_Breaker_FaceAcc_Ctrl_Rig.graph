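# Control Rig for the female medium Heart Breaker face accessory, offsetting hat and hair bones according to head pose.
# A SphericalPoseReader on the head bone, with debug drawing on, outputs a falloff weight.
# That weight blends two chained ModifyTransforms steps: one for Hat, the next for hair base bones including the ponytail and right bang.
# Offsets are literal pin values; the visible graph reads none of the declared transform variables.
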ControlRigBlueprint /BRCosmetics/Characters/Player/Female/Medium/Bodies/F_MED_Heart_Breaker/Meshes/Parts/F_MED_Heart_Breaker_FaceAcc_Ctrl_Rig
graphs [
 RigVMFunctionLibrary {links [], nodes []}
 RigVMModel {nodes [
   RigUnit_BeginExecution = &RigUnit_BeginExecution()
   n1 = &RigUnit_SphericalPoseReader(
    ActiveRegionScaleFactors=(PositiveWidth=1,NegativeWidth=1,PositiveHeight=1,NegativeHeight=1)
    ActiveRegionSize=0.1
    Debug=(bDrawDebug=True,bDraw2D=False,bDrawLocalAxes=False,DebugScale=25,DebugSegments=20,DebugThickness=0.25)
    DriverAxis=(X=1,Y=0,Z=0)
    DriverItem=(Type=Bone,Name="head")
    FalloffRegionScaleFactors=(PositiveWidth=1,NegativeWidth=1,PositiveHeight=1,NegativeHeight=1)
    FalloffSize=0.18
    FlipHeightScaling=false
    FlipWidthScaling=false
    OptionalParentItem=(Type=Bone)
    RotationOffset=(X=90,Y=0,Z=-135)
    ExecutePin=@RigUnit_BeginExecution.ExecutePin
   )
   > &RigUnit_ModifyTransforms(
    ItemToModify=((Item=(Type=Bone,Name="Hat"),Transform=(Rotation=(X=0,Y=0,Z=0.0872,W=0.9962),Translation=(X=2,Y=0,Z=0),Scale3D=(X=1,Y=1,Z=1))))
    Weight=@n1.OutputParam
   )
   > &RigUnit_ModifyTransforms(
    ItemToModify="((Item=(Type=Bone,Name=\"dyn_ponytail_base\"),Transform=(Rotation=(X=0,Y=0.3827,Z=0,W=0.9239),Translation=(X=0,Y=0,Z=-0.5),Scale3D=(X=1,Y=1,Z=1))),(Item=(Type=Bone,Name=\"dyn_bang_r_base\"),Transform=(Rotation=(X=0,Y=0,Z=0,W=1),Translation=(X=0,Y=0,Z=0),Scale3 ...(truncated)"
    Weight=@n1.OutputParam
   )
  ]}
]
hierarchy [
 root {loc=<0 0 0>, rot=<0 0 0 1>}
 pelvis {parent=root, loc=<0 -0.5301 98.0571>, rot=<0.0127 0.707 -0.0127 -0.707>}
 spine_01 {parent=pelvis, loc=<6.1208 -0.9328 0>, rot=<0 0 0.034 -0.9994>}
 spine_02 {parent=spine_01, loc=<6.5996 -0.2095 0>, rot=<0 0 0.0238 -0.9997>}
 spine_03 {parent=spine_02, loc=<6.4914 -0.1237 0>, rot=<0 0 0.0149 -0.9999>}
 spine_04 {parent=spine_03, loc=<6.9598 0.0794 0>, rot=<0 0 -0.0132 -0.9999>}
 spine_05 {parent=spine_04, loc=<6.9987 0 0>, rot=<0 0 -0.0015 -1>}
 neck_01 {parent=spine_05, loc=<11.3138 0.1529 0>, rot=<0 0 -0.2003 -0.9797>}
 neck_02 {parent=neck_01, loc=<4.5627 0 0>, rot=<0 0 0.0504 -0.9987>}
 head {parent=neck_02, loc=<5.3914 0 0>, rot=<0 0 0.115 -0.9934>}
 dyn_ponytail_base {parent=head, loc=<1.8567 -9.3641 0.012>, rot=<-0.4007 0.5703 0.6118 -0.374>}
 dyn_ponytail_parent {parent=dyn_ponytail_base, loc=<0 0 0>, rot=<0 0 0 1>}
 dyn_ponytail_01 {parent=dyn_ponytail_parent, loc=<2.8974 0.162 0.8107>, rot=<0.0132 0.4226 0.0114 -0.9062>}
 dyn_ponytail_02 {parent=dyn_ponytail_01, loc=<2.9419 0.0184 0.6617>, rot=<-0.0518 0.139 -0.0073 -0.9889>}
 dyn_ponytail_03 {parent=dyn_ponytail_02, loc=<3.142 0 0.4018>, rot=<0 0 0 1>}
 dyn_ponytail_strand_02 {parent=dyn_ponytail_01, loc=<3.3623 -1.1808 -1.7638>, rot=<-0.0259 -0.1164 -0.058 -0.9912>}
 dyn_ponytail_strand_01 {parent=dyn_ponytail_01, loc=<0.8822 -3.4678 -1.6637>, rot=<0.1405 -0.3368 0.1021 -0.9254>}
 dyn_bang_r_base {parent=head, loc=<0.5906 1.1593 6.5052>, rot=<0.0041 -1 0 0>}
 dyn_bang_r_parent {parent=dyn_bang_r_base, loc=<0 0 0>, rot=<0 0 0 1>}
 dyn_bang_01_r {parent=dyn_bang_r_parent, loc=<2.6627 -0.0217 -0.2536>, rot=<0 -0.2756 0 -0.9613>}
 dyn_bang_l_base {parent=head, loc=<0.2158 1.266 -6.4011>, rot=<-0.0001 0.0175 0.9998 -0.0041>}
 dyn_bang_l_parent {parent=dyn_bang_l_base, loc=<0 0 0>, rot=<0 0 0 1>}
 dyn_bang_l_01 {parent=dyn_bang_l_parent, loc=<2.0014 -0.3575 -0.1343>, rot=<0.0172 -0.1736 0.003 -0.9847>}
 Hat {parent=head, loc=<0 0 0>, rot=<0 0 0 1>}
]
variables [
 Hair Middle L Transform {type=FTransform, default=(Rotation=(X=0,Y=-0.1908,Z=0,W=0.9816),Translation=(X=0,Y=0,Z=0),Scale3D=(X=1,Y=1,Z=1))}
 Hair Middle R Transform {type=FTransform, default=(Rotation=(X=-0.2541,Y=-0.1843,Z=-0.0494,W=0.9482),Translation=(X=0,Y=0,Z=0),Scale3D=(X=1,Y=1,Z=1))}
 Hair Out R Transform {type=FTransform, default=(Rotation=(X=-0.1913,Y=-0.4619,Z=0.1913,W=0.8446),Translation=(X=0,Y=0,Z=0),Scale3D=(X=1,Y=1,Z=1))}
 Hair Out R Transform_0 {type=FTransform, default=(Rotation=(X=-0.0782,Y=-0.0579,Z=0.1691,W=0.9808),Translation=(X=0,Y=0,Z=0),Scale3D=(X=1,Y=1,Z=1))}
 Hair Out R Transform_1 {type=FTransform, default=(Rotation=(X=0,Y=0,Z=0,W=1),Translation=(X=0,Y=1.25,Z=0),Scale3D=(X=1,Y=1,Z=1))}
 ItemToModify_0_Transform {type=FTransform, default=(Rotation=(X=0,Y=0,Z=0,W=1),Translation=(X=0,Y=0,Z=0),Scale3D=(X=1,Y=1,Z=1))}
 ItemToModify_0_Transform_Rotation {type=FQuat, default=(X=0,Y=0,Z=-0.5,W=0.866)}
]
