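# Physics asset for the Heart Breaker medium female character part, covering the dynamic bones of its bangs and ponytail.
# Left and right bangs use short chains of boxes, while the ponytail is a capsule chain with two strands branching from its 01 bone.
# Swing limits and tight twist limits join each chain up to its base bone.
# Unconstrained head and upper-spine capsules supply body shapes, likely as colliders for the hair.
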PhysicsAsset /BRCosmetics/Characters/Player/Female/Medium/Bodies/F_MED_Heart_Breaker/Meshes/Parts/F_MED_Heart_Breaker_Physics
bodies {
 dyn_bang_01_r [{shapeType=Box, extentX=1.582, extentY=2.2162, extentZ=1, center=<0.7254 0 -0.0001>}]
 dyn_bang_l_01 [{shapeType=Box, extentX=2.26, extentY=2.26, extentZ=0.76, center=<1.2151 0 0>}]
 dyn_bang_l_base [{shapeType=Box, extentX=0.5, extentY=0.5, extentZ=0.5}]
 dyn_bang_l_parent [{shapeType=Box, extentX=2.3482, extentY=2.5293, extentZ=0.75, center=<1.0388 0 0>}]
 dyn_bang_r_base [{shapeType=Box, extentX=0.5, extentY=0.5, extentZ=0.5}]
 dyn_bang_r_parent [{shapeType=Box, extentX=3.187, extentY=3.021, extentZ=0.75, center=<1.4703 0 0>}]
 dyn_ponytail_01 [{shapeType=Capsule, length=1.642, radius=1.3425, center=<1.9215 -0.0001 -0.5406>, rotation=<rot -74.2848 -179.9999 180>}]
 dyn_ponytail_02 [{shapeType=Capsule, length=1.1373, radius=1.5963, center=<1.9899 0 0>, rotation=<rot 90 0 0>}]
 dyn_ponytail_03 [{shapeType=Capsule, length=2.3367, radius=1.3372, center=<1.9803 0 0>, rotation=<rot 90 0 0>}]
 dyn_ponytail_base [{shapeType=Capsule, length=0.505, radius=0.505}]
 dyn_ponytail_parent [{shapeType=Capsule, length=1.4893, radius=1.4666, center=<1.5889 0 -0.0962>, rotation=<rot -86.5343 -179.9998 179.9998>}]
 dyn_ponytail_strand_01 [{shapeType=Capsule, length=1.5276, radius=0.5686, center=<1.302 0 0>, rotation=<rot 90 0 0>}]
 dyn_ponytail_strand_02 [{shapeType=Capsule, length=1.5941, radius=0.8368, center=<1.2731 0 0>, rotation=<rot 90 0 0>}]
 head [{shapeType=Capsule, length=2.875, radius=2.875, rotation=<rot 90 0 0>}]
 spine_05 [{shapeType=Capsule, length=20.125, radius=9.875, center=<5.2065 3.4918 0>}]
}
constraints [
 {bone1Name=dyn_bang_01_r, bone2Name=dyn_bang_r_parent, swing1LimitDegrees=20, swing2LimitDegrees=20, twistLimitDegrees=3}
 {bone1Name=dyn_bang_r_parent, bone2Name=dyn_bang_r_base, swing1LimitDegrees=30, swing2LimitDegrees=20, twistLimitDegrees=3}
 {bone1Name=dyn_bang_l_01, bone2Name=dyn_bang_l_parent, swing1LimitDegrees=20, swing2LimitDegrees=20, twistLimitDegrees=3}
 {bone1Name=dyn_bang_l_parent, bone2Name=dyn_bang_l_base, swing1LimitDegrees=30, swing2LimitDegrees=20, twistLimitDegrees=3}
 {bone1Name=dyn_ponytail_parent, bone2Name=dyn_ponytail_base, swing1LimitDegrees=20, swing2LimitDegrees=20, twistLimitDegrees=3}
 {bone1Name=dyn_ponytail_01, bone2Name=dyn_ponytail_parent, swing1LimitDegrees=20, swing2LimitDegrees=20, twistLimitDegrees=3}
 {bone1Name=dyn_ponytail_02, bone2Name=dyn_ponytail_01, swing1LimitDegrees=20, swing2LimitDegrees=15, twistLimitDegrees=3}
 {bone1Name=dyn_ponytail_03, bone2Name=dyn_ponytail_02, swing1LimitDegrees=20, swing2LimitDegrees=15, twistLimitDegrees=3}
 {bone1Name=dyn_ponytail_strand_02, bone2Name=dyn_ponytail_01, swing1LimitDegrees=20, swing2LimitDegrees=20, twistLimitDegrees=3}
 {bone1Name=dyn_ponytail_strand_01, bone2Name=dyn_ponytail_01, swing1LimitDegrees=20, swing2LimitDegrees=20, twistLimitDegrees=3}
]
